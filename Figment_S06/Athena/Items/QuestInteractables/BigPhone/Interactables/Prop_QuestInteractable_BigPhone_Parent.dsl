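; Parent Blueprint for the Big Phone quest interactable prop in Fortnite's Athena mode, judging by its name and path.
; It adds no logic of its own in the graphs shown.
; BeginPlay, quest initialization, objective completion, quest updates and the construction script all forward straight to the inherited implementation, so behavior comes from its parent class.
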(event EventBeginPlay
  ;[1 1 1] EVENT -> BeginPlay
  (|CallParentFunction))

(event EventQuestInitializedAndValid
  ;[1 1 1] EVENT -> QuestInitializedAndValid
  (|CallParentFunction))

(event EventThisQuestObjectiveComplete
  (params (IneractingPawn :type Fort%20Pawn%20Object%20Reference))
  ;[1 1 1] EVENT -> ThisQuestObjectiveComplete
  (|CallParentFunction IneractingPawn))

(event EventQuestUpdatedFromAnywhere
  ;[1 1 1] EVENT -> QuestUpdatedFromAnywhere
  (|CallParentFunction))

(fn ConstructionScript
  (|CallParentFunction))
